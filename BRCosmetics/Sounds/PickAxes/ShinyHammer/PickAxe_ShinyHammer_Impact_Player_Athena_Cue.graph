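# Sound cue for the ShinyHammer cosmetic pickaxe's player impact sound in Athena.
# A short fixed delay feeds a Random node that picks among six equally weighted impact wave variations.
# It uses the shared pickaxe impact attenuation and cosmetic property sheet, the Fort_Impacts sound class, and the player melee weapon concurrency set.
# The cue is primed on load.
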SoundCue /BRCosmetics/Sounds/PickAxes/ShinyHammer/PickAxe_ShinyHammer_Impact_Player_Athena_Cue
root=Delay_1
nodes [
 {id=Delay_1, children [Random_0], properties {delayMax=0.05, delayMin=0.05}}
 {
  id=Random_0
  children [WavePlayer_11, WavePlayer_12, WavePlayer_13, WavePlayer_14, WavePlayer_15, WavePlayer_16]
  properties {weights [1,1,1,1,1,1]}
 }
 {id=WavePlayer_11, properties {soundWaveAssetPtr=&/Athena_Pickaxe_ShinyHammer_Impact_01}}
 {id=WavePlayer_12, properties {soundWaveAssetPtr=&/Athena_Pickaxe_ShinyHammer_Impact_02}}
 {id=WavePlayer_13, properties {soundWaveAssetPtr=&/Athena_Pickaxe_ShinyHammer_Impact_03}}
 {id=WavePlayer_14, properties {soundWaveAssetPtr=&/Athena_Pickaxe_ShinyHammer_Impact_04}}
 {id=WavePlayer_15, properties {soundWaveAssetPtr=&/Athena_Pickaxe_ShinyHammer_Impact_05}}
 {id=WavePlayer_16, properties {soundWaveAssetPtr=&/Athena_Pickaxe_ShinyHammer_Impact_06}}
]
properties {
 attenuationSettings=/Game/Sounds/Attenuation/Fort_Pickaxe_Impact_Att
 audioPropertiesSheet=/BRCosmetics/Sounds/Settings/PropertySheets/Pickaxe/APS_Cosmetic_Pickaxe_Impact
 bPrimeOnLoad=true
 maxDistance=4300
 soundClassObject=&Fort_Impacts
 attenuationOverrides {
  bAttenuateWithLPF=true
  distanceAlgorithm=Logarithmic
  falloffDistance=800
  lPFRadiusMax=800
  lPFRadiusMin=400
  attenuationShapeExtents {x=300}
 }
 concurrencyOverrides {bLimitToOwner=true, maxCount=2, resolutionRule=StopOldest}
 concurrencySet [&Player_Melee_Weapon_Conc]
}
